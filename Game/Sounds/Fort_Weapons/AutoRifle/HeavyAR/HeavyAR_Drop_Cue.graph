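# Sound cue for dropping a heavy assault rifle.
# A Random node picks with equal weight among four generic heavy-rifle drop waves, which live in the UI gameplay sounds folder.
# The cue uses the shared weapon-drop attenuation, the loot-collect sound class and a boosted volume.
# It joins the Weapon_Drop_StopOldest concurrency set to manage overlapping drops.
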SoundCue /Game/Sounds/Fort_Weapons/AutoRifle/HeavyAR/HeavyAR_Drop_Cue
root=Random_0
nodes [
 {id=Random_0, children [WavePlayer_1,WavePlayer_3,WavePlayer_4,WavePlayer_5], properties {weights [1,1,1,1]}}
 {id=WavePlayer_1, properties {soundWaveAssetPtr=/Game/Sounds/Fort_UI_Sounds/Gameplay/Weapon_Generic_Drop_Rifle_Hv_01}}
 {id=WavePlayer_3, properties {soundWaveAssetPtr=/Game/Sounds/Fort_UI_Sounds/Gameplay/Weapon_Generic_Drop_Rifle_Hv_02}}
 {id=WavePlayer_4, properties {soundWaveAssetPtr=/Game/Sounds/Fort_UI_Sounds/Gameplay/Weapon_Generic_Drop_Rifle_Hv_03}}
 {id=WavePlayer_5, properties {soundWaveAssetPtr=/Game/Sounds/Fort_UI_Sounds/Gameplay/Weapon_Generic_Drop_Rifle_Hv_04}}
]
properties {
 attenuationSettings=&Fort_Weapon_Drop_Att
 maxDistance=3000
 soundClassObject=&Fort_Loot_Collect
 volumeMultiplier=1.25
 concurrencyOverrides {resolutionRule=PreventNew}
 concurrencySet [&Weapon_Drop_StopOldest]
}
