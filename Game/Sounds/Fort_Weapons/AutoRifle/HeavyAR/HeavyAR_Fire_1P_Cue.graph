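# First-person gunfire sound cue for the Heavy AR weapon.
# A Random node picks one of four equally weighted close-fire waves and feeds it through a Modulator for playback variation, with the pitch minimum set to unity.
# The cue plays at reduced volume through a first-person ranged-fire sound class, with first-person projectile attenuation and a gunfire concurrency set, and has its debug flag enabled.
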SoundCue /Game/Sounds/Fort_Weapons/AutoRifle/HeavyAR/HeavyAR_Fire_1P_Cue
root=Modulator_0
nodes [
 {id=Modulator_0, children [Random_2], properties {pitchMin=1}}
 {id=Random_2, children [WavePlayer_9, WavePlayer_11, WavePlayer_12, WavePlayer_13], properties {weights [1,1,1,1]}}
 {id=WavePlayer_11, properties {soundWaveAssetPtr=&/HeavyAR_Fire_Close_02}}
 {id=WavePlayer_12, properties {soundWaveAssetPtr=&/HeavyAR_Fire_Close_03}}
 {id=WavePlayer_13, properties {soundWaveAssetPtr=&/HeavyAR_Fire_Close_04}}
 {id=WavePlayer_9, properties {soundWaveAssetPtr=&/HeavyAR_Fire_Close_01}}
]
properties {
 attenuationSettings=&Fort_Weapon_Projectile_FirstPerson_Athena_Att
 bDebug=true
 maxDistance=2097152
 soundClassObject=&Fort_WeaponRanged_Fire_1P_PSM
 volumeMultiplier=0.7
 concurrencySet [&Weapon_Gunfire_FirstPerson_Conc]
}
