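# Sound cue for the Heavy AR's gunfire, layered by distance into close, medium and distant shots.
# Each layer picks a random wave, passes it through a modulator and applies its own attenuation asset.
# The close layer uses a third-person fire sound class, while medium and distant are mixed under a distant-fire class before the final mixer.
# Local and global gunfire concurrency sets govern simultaneous shots, and debug is switched on.
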SoundCue /Game/Sounds/Fort_Weapons/AutoRifle/HeavyAR/HeavyAR_Fire_Cue
root=Mixer_1
nodes [
 {
  id=Attenuation_118
  children [Modulator_4]
  properties {attenuationSettings=&Fort_Weapon_Projectile_Close_Athena_Att}
 }
 {
  id=Attenuation_119
  children [Modulator_2]
  properties {attenuationSettings=&Fort_Weapon_Projectile_Medium_Mono_Athena_Att}
 }
 {
  id=Attenuation_120
  children [Modulator_1]
  properties {attenuationSettings=&Fort_Weapon_Projectile_Distant_Mono_Athena_Att}
 }
 {id=Mixer_0, children [Attenuation_119, Attenuation_120], properties {inputVolume [0.8,1]}}
 {id=Mixer_1, children [SoundClass_77, SoundClass_76], properties {inputVolume [0.95,0.75]}}
 {id=Modulator_1, children [Random_5], properties {pitchMax=1.1, pitchMin=0.9, volumeMax=0.8, volumeMin=0.7}}
 {id=Modulator_2, children [Random_4], properties {volumeMax=1.7, volumeMin=1.6}}
 {id=Modulator_4, children [Random_1], properties {pitchMin=1}}
 {id=Random_1, children [WavePlayer_1,WavePlayer_2,WavePlayer_3,WavePlayer_4], properties {weights [1,1,1,1]}}
 {id=Random_4, children [WavePlayer_10, WavePlayer_11, WavePlayer_12], properties {weights [1,1,1]}}
 {id=Random_5, children [WavePlayer_16, WavePlayer_17, WavePlayer_20], properties {weights [1,1,1]}}
 {id=SoundClass_76, children [Mixer_0], properties {soundClassOverride=&Fort_WeaponRanged_Fire_Distant}}
 {
  id=SoundClass_77
  children [Attenuation_118]
  properties {soundClassOverride=&Fort_WeaponRanged_Fire_3P_PSM}
 }
 {id=WavePlayer_1, properties {soundWaveAssetPtr=&/HeavyAR_Fire_Close_01}}
 {id=WavePlayer_10, properties {soundWaveAssetPtr=&/HeavyAR_Fire_Medium_01}}
 {id=WavePlayer_11, properties {soundWaveAssetPtr=&/HeavyAR_Fire_Medium_02}}
 {id=WavePlayer_12, properties {soundWaveAssetPtr=&/HeavyAR_Fire_Medium_03}}
 {id=WavePlayer_16, properties {soundWaveAssetPtr=&/HeavyAR_Fire_Distant_01}}
 {id=WavePlayer_17, properties {soundWaveAssetPtr=&/HeavyAR_Fire_Distant_02}}
 {id=WavePlayer_2, properties {soundWaveAssetPtr=&/HeavyAR_Fire_Close_02}}
 {id=WavePlayer_20, properties {soundWaveAssetPtr=&/HeavyAR_Fire_Distant_03}}
 {id=WavePlayer_3, properties {soundWaveAssetPtr=&/HeavyAR_Fire_Close_03}}
 {id=WavePlayer_4, properties {soundWaveAssetPtr=&/HeavyAR_Fire_Close_04}}
]
properties {
 bDebug=true
 maxDistance=27500
 pitchMultiplier=0.95
 soundClassObject=&Fort_WeaponRanged
 concurrencySet [&Weapon_Gunfire_Local_Conc, &Weapon_Gunfire_Global_Conc]
}
