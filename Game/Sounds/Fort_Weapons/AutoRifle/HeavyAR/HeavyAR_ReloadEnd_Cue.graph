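# Sound cue for the end of the heavy assault rifle's reload, classed as weapon foley.
# A single wave player feeds the root directly, with no randomization or mixing.
# The cue slightly lowers pitch and volume and uses the shared weapon reload attenuation.
# Concurrency override and the debug flag are both enabled.
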SoundCue /Game/Sounds/Fort_Weapons/AutoRifle/HeavyAR/HeavyAR_ReloadEnd_Cue
root=WavePlayer_1
nodes [{id=WavePlayer_1, properties {soundWaveAssetPtr=&/HeavyAR_ReloadEnd_01}}]
properties {
 attenuationSettings=&Fort_Weapon_Reload_Att
 bDebug=true
 bOverrideConcurrency=true
 maxDistance=2250
 pitchMultiplier=0.95
 soundClassObject=&Fort_WeaponFoley
 volumeMultiplier=0.752
}
