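# Sound cue for the Heavy AR weapon's magazine insert during reload.
# A random node picks between two insert wave variations with equal weight.
# It routes to the weapon foley sound class with reload attenuation and a reduced volume, overriding concurrency.
# The debug flag is left switched on.
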SoundCue /Game/Sounds/Fort_Weapons/AutoRifle/HeavyAR/HeavyAR_ReloadInsert_Cue
root=Random_0
nodes [
 {id=Random_0, children [WavePlayer_2,WavePlayer_3], properties {weights [1,1]}}
 {id=WavePlayer_2, properties {soundWaveAssetPtr=&/HeavyAR_ReloadInsert_01}}
 {id=WavePlayer_3, properties {soundWaveAssetPtr=&/HeavyAR_ReloadInsert_02}}
]
properties {
 attenuationSettings=&Fort_Weapon_Reload_Att
 bDebug=true
 bOverrideConcurrency=true
 maxDistance=2250
 soundClassObject=&Fort_WeaponFoley
 volumeMultiplier=0.626
}
